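# Physics asset for the Minotaur Luck back bling mesh, providing the bodies and joints for its four chains: top and bottom on each side.
# Each chain has a three-capsule A strand and a two-capsule B strand, both rooted to the small pack body with tight swing and twist limits.
# Constraints with locked angular motion join each B strand's tip to its A strand's tip, closing every chain into a loop anchored on the pack.
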PhysicsAsset /BRCosmetics/Accessories/FORT_Backpacks/Backpack_M_MED_Minotaur_Luck/Meshes/M_MED_Minotaur_Luck_Physics
bodies {
 dyn_chainBottomA1_l [{shapeType=Capsule, length=1, radius=1, center=<1.5 0 0>, rotation=<rot 90 0 0>}]
 dyn_chainBottomA1_r [{shapeType=Capsule, length=1, radius=1, center=<1.5 0 0>, rotation=<rot 90 0 0>}]
 dyn_chainBottomA2_l [{shapeType=Capsule, length=1, radius=1, center=<1.5 0 0>, rotation=<rot 90 0 0>}]
 dyn_chainBottomA2_r [{shapeType=Capsule, length=1, radius=1, center=<1.5 0 0>, rotation=<rot 90 0 0>}]
 dyn_chainBottomA3_l [{shapeType=Capsule, length=1, radius=1, center=<1.5 0 0>, rotation=<rot 90 0 0>}]
 dyn_chainBottomA3_r [{shapeType=Capsule, length=1, radius=1, center=<1.5 0 0>, rotation=<rot 90 0 0>}]
 dyn_chainBottomB1_l [{shapeType=Capsule, length=1, radius=1, center=<1.5 0 0>, rotation=<rot 90 0 0>}]
 dyn_chainBottomB1_r [{shapeType=Capsule, length=1, radius=1, center=<1.5 0 0>, rotation=<rot 90 0 0>}]
 dyn_chainBottomB2_l [{shapeType=Capsule, length=1, radius=1, center=<1.5 0 0>, rotation=<rot 90 0 0>}]
 dyn_chainBottomB2_r [{shapeType=Capsule, length=1, radius=1, center=<1.5 0 0>, rotation=<rot 90 0 0>}]
 dyn_chainTopA1_l [{shapeType=Capsule, length=1, radius=1, center=<1.5 0 0>, rotation=<rot 90 0 0>}]
 dyn_chainTopA1_r [{shapeType=Capsule, length=1, radius=1, center=<1.5 0 0>, rotation=<rot 90 0 0>}]
 dyn_chainTopA2_l [{shapeType=Capsule, length=1, radius=1, center=<1.5 0 0>, rotation=<rot 90 0 0>}]
 dyn_chainTopA2_r [{shapeType=Capsule, length=1, radius=1, center=<1.5 0 0>, rotation=<rot 90 0 0>}]
 dyn_chainTopA3_l [{shapeType=Capsule, length=1, radius=1, center=<1.5 0 0>, rotation=<rot 90 0 0>}]
 dyn_chainTopA3_r [{shapeType=Capsule, length=1, radius=1, center=<1.5 0 0>, rotation=<rot 90 0 0>}]
 dyn_chainTopB1_l [{shapeType=Capsule, length=1, radius=1, center=<1.5 0 0>, rotation=<rot 90 0 0>}]
 dyn_chainTopB1_r [{shapeType=Capsule, length=1, radius=1, center=<1.5 0 0>, rotation=<rot 90 0 0>}]
 dyn_chainTopB2_l [{shapeType=Capsule, length=1, radius=1, center=<1.5 0 0>, rotation=<rot 90 0 0>}]
 dyn_chainTopB2_r [{shapeType=Capsule, length=1, radius=1, center=<1.5 0 0>, rotation=<rot 90 0 0>}]
 pack [{shapeType=Capsule, length=0, radius=0.5, center=<0 -5.5194 -9.1244>}]
}
constraints [
 {bone1Name=dyn_chainTopA1_r, bone2Name=pack, swing1LimitDegrees=2, swing2LimitDegrees=10, twistLimitDegrees=5}
 {bone1Name=dyn_chainTopA2_r, bone2Name=dyn_chainTopA1_r, swing1LimitDegrees=2, swing2LimitDegrees=10, twistLimitDegrees=5}
 {bone1Name=dyn_chainTopA3_r, bone2Name=dyn_chainTopA2_r, swing1LimitDegrees=2, swing2LimitDegrees=10, twistLimitDegrees=5}
 {bone1Name=dyn_chainTopB1_r, bone2Name=pack, swing1LimitDegrees=2, swing2LimitDegrees=10, twistLimitDegrees=5}
 {bone1Name=dyn_chainTopB2_r, bone2Name=dyn_chainTopB1_r, swing1LimitDegrees=2, swing2LimitDegrees=10, twistLimitDegrees=5}
 {bone1Name=dyn_chainBottomA1_r, bone2Name=pack, swing1LimitDegrees=2, swing2LimitDegrees=10, twistLimitDegrees=5}
 {bone1Name=dyn_chainBottomA2_r, bone2Name=dyn_chainBottomA1_r, swing1LimitDegrees=2, swing2LimitDegrees=10, twistLimitDegrees=5}
 {bone1Name=dyn_chainBottomA3_r, bone2Name=dyn_chainBottomA2_r, swing1LimitDegrees=2, swing2LimitDegrees=10, twistLimitDegrees=5}
 {bone1Name=dyn_chainBottomB1_r, bone2Name=pack, swing1LimitDegrees=2, swing2LimitDegrees=10, twistLimitDegrees=5}
 {bone1Name=dyn_chainBottomB2_r, bone2Name=dyn_chainBottomB1_r, swing1LimitDegrees=2, swing2LimitDegrees=10, twistLimitDegrees=5}
 {bone1Name=dyn_chainTopA1_l, bone2Name=pack, swing1LimitDegrees=2, swing2LimitDegrees=10, twistLimitDegrees=5}
 {bone1Name=dyn_chainTopA2_l, bone2Name=dyn_chainTopA1_l, swing1LimitDegrees=2, swing2LimitDegrees=10, twistLimitDegrees=5}
 {bone1Name=dyn_chainTopA3_l, bone2Name=dyn_chainTopA2_l, swing1LimitDegrees=2, swing2LimitDegrees=10, twistLimitDegrees=5}
 {bone1Name=dyn_chainTopB1_l, bone2Name=pack, swing1LimitDegrees=2, swing2LimitDegrees=10, twistLimitDegrees=5}
 {bone1Name=dyn_chainTopB2_l, bone2Name=dyn_chainTopB1_l, swing1LimitDegrees=2, swing2LimitDegrees=10, twistLimitDegrees=5}
 {bone1Name=dyn_chainBottomA1_l, bone2Name=pack, swing1LimitDegrees=2, swing2LimitDegrees=10, twistLimitDegrees=5}
 {bone1Name=dyn_chainBottomA2_l, bone2Name=dyn_chainBottomA1_l, swing1LimitDegrees=2, swing2LimitDegrees=10, twistLimitDegrees=5}
 {bone1Name=dyn_chainBottomA3_l, bone2Name=dyn_chainBottomA2_l, swing1LimitDegrees=2, swing2LimitDegrees=10, twistLimitDegrees=5}
 {bone1Name=dyn_chainBottomB1_l, bone2Name=pack, swing1LimitDegrees=2, swing2LimitDegrees=10, twistLimitDegrees=5}
 {bone1Name=dyn_chainBottomB2_l, bone2Name=dyn_chainBottomB1_l, swing1LimitDegrees=2, swing2LimitDegrees=10, twistLimitDegrees=5}
 {
  bone1Name=dyn_chainBottomB2_r
  bone2Name=dyn_chainBottomA3_r
  swing1LimitDegrees=45
  swing1Motion=Locked
  swing2LimitDegrees=45
  swing2Motion=Locked
  twistLimitDegrees=45
  twistMotion=Locked
 }
 {
  bone1Name=dyn_chainTopB2_l
  bone2Name=dyn_chainTopA3_l
  swing1LimitDegrees=45
  swing1Motion=Locked
  swing2LimitDegrees=45
  swing2Motion=Locked
  twistLimitDegrees=45
  twistMotion=Locked
 }
 {
  bone1Name=dyn_chainBottomB2_l
  bone2Name=dyn_chainBottomA3_l
  swing1LimitDegrees=45
  swing1Motion=Locked
  swing2LimitDegrees=45
  swing2Motion=Locked
  twistLimitDegrees=45
  twistMotion=Locked
 }
 {
  bone1Name=dyn_chainTopB2_r
  bone2Name=dyn_chainTopA3_r
  swing1LimitDegrees=45
  swing1Motion=Locked
  swing2LimitDegrees=45
  swing2Motion=Locked
  twistLimitDegrees=45
  twistMotion=Locked
 }
]
